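# Main Niagara effect for the SkyRuled emote: a black hole with rings, orbit and connection meshes, swirling stones and purple smoke.
# Emitters follow the character through a shared skeletal mesh reader, with system SizeAll, OrbitSize and FXOffset driving scale and placement.
# Vortex and attraction forces swirl smoke and GPU stones, while movement-spawned world-space smoke trails behind.
# Distance remaps and front-end modifiers tune opacity and density; the original BlackHole emitter is disabled, replaced by BlackHole001.
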NiagaraSystem /BRCosmetics/Effects/Fort_Effects/Emotes/SkyRuled/NS_Emote_SkyRuled_Main
dependencies {
 usedDataInterfaces [
  &NiagaraDataInterfaceCurve
  &NiagaraDataInterfaceArrayFloat
  &NiagaraDataInterfaceArrayFloat3
  &NiagaraDataInterfaceSkeletalMesh
 ]
 usedDynamicInputs [&ReturnNormalizedExecIndex]
 usedModules [
  &self:SystemSpawnScript.SystemScriptSource.SystemScriptGraph.NiagaraNodeAssignment_1.SetVariables_#1
  &self:SystemSpawnScript.SystemScriptSource.SystemScriptGraph.NiagaraNodeAssignment_2.SetVariables_#2
  &self:SystemSpawnScript.SystemScriptSource.SystemScriptGraph.NiagaraNodeAssignment_3.SetVariables_#3
  &self:SystemSpawnScript.SystemScriptSource.SystemScriptGraph.NiagaraNodeAssignment_0.SetVariables_#4
  &SystemState
  &InterpolateOverTime
 ]
}
emitters {
 BlackHole {
  bEnabled=false
  emitterUpdateScript [
   &EmitterState(Life Cycle Mode=::NewEnumerator1, Loop Behavior=::NewEnumerator1, Loop Duration Mode=::NewEnumerator1, UseLoopDelay=false)
   &SpawnBurst_Instantaneous(Use Spawn Probability=false, UseLoopCountLimit=false)
  ]
  particleSpawnScript [
   &InitializeParticle(
    Color=<rgba 0 0 0.2135 7>
    Position Offset=&AddVectorToPosition(Position=Linked() {linkedVariable {name=System.FXOffset, type=Vector3f}}, Vector=<0 0 0>)
    Sprite Size Mode=::NewEnumerator3
    Uniform Sprite Size=120
    UsePositionOffset=false
    Write Material Random=true
   )
   &DynamicMaterialParameters(
    Index 0 Param 1=2
    Index 0 Param 2=10
    Module.Index0Float0Write=true
    Module.Index0Float1Write=true
    Module.Index0Float2Write=true
    Module.Index0Float3Write=true
   )
   &CameraOffset()
  ]
  particleUpdateScript [
   &ParticleState(Kill Particles When Lifetime Has Elapsed=false)
   &SpriteRotationRate(Rotation Rate=120)
   &ScaleSpriteSize(Scale Sprite Size Mode=::NewEnumerator0, Uniform Scale Factor=$System.SizeAll)
   &SkeletalMeshLocation(Bone Sampling Mode=::NewEnumerator4, MaskSpawnGroup=false, Sampled Position Offset=<50 0 0>, Skeletal Mesh=$System.Skeletal Mesh Reader)
   &FN_Sprite_Attributes(
    Camera Offset=30
    Color=&MakeLinearColorFromVectorAndFloat(
     "Vector (RGB)"=&Multiply_Vector(A=&MakeVectorFromLinearColorRGB(Color=<rgba 3.7917 12.3035 40 2>), B=<1 1 1>)
     "Float (Alpha)"=&Multiply_Float(A=2, B=0.5)
    )
    Create Camera Offset Attribute=true
    Create Color Attribute=true
    Create SpriteSize Attribute=true
    Module Output:ENiagara_AttributeSamplingApplyOutput=::NewEnumerator1
    SpriteSize=&Multiply_Vector2D(
     A=&Vector2DFromFloat(Value=73)
     B=&Multiply_Vector2D_ByFloat(Float=$System.SizeAll, Vector2D=<1 1>)
    )
   )
   &FN_Mesh_Attributes(
    Color=&MakeLinearColorFromVectorAndFloat("Vector (RGB)"=<0 0 0>, "Float (Alpha)"=1)
    Create Color Attribute=true
    Create Scale Attribute=true
    Module Output=::NewEnumerator1
    Scale=&Multiply_VectorByFloat(Vector=&Multiply_VectorByFloat(Vector=<1 1 1>, Float=$System.SizeAll), Float=0.195)
   )
   &FN_Mesh_Attributes(
    Color=&MakeLinearColorFromVectorAndFloat("Vector (RGB)"=&MakeVectorFromLinearColorRGB(Color=<rgba 1.46 0.3768 1.5 1>), "Float (Alpha)"=3)
    Create Color Attribute=true
    Create Scale Attribute=true
    Module Output=::NewEnumerator1
    Scale=&Multiply_VectorByFloat(Vector=&Multiply_VectorByFloat(Vector=<2.5 2.5 0.2>, Float=$System.SizeAll), Float=0.2)
   ) {module=FN_Mesh_Attributes001}
  ]
  rendererClasses [&NiagaraSpriteRendererProperties, &NiagaraMeshRendererProperties]
  renderers [
   {rendererClass=&NiagaraSpriteRendererProperties}
   {rendererClass=&NiagaraSpriteRendererProperties, rendererIndex=1}
   {rendererClass=&NiagaraMeshRendererProperties, rendererIndex=2}
   {rendererClass=&NiagaraMeshRendererProperties, rendererIndex=3}
  ]
 }
 BlackHole001 {
  emitterUpdateScript [
   &EmitterState(Life Cycle Mode=::NewEnumerator1, Loop Behavior=::NewEnumerator1, Loop Duration Mode=::NewEnumerator1, UseLoopDelay=false)
   &SpawnBurst_Instantaneous(Use Spawn Probability=false, UseLoopCountLimit=false)
  ]
  particleSpawnScript [
   &InitializeParticle(
    Color=<rgba 0 0 0.2135 7>
    Position Offset=&AddVectorToPosition(Position=Linked() {linkedVariable {name=System.FXOffset, type=Vector3f}}, Vector=<0 0 0>)
    Sprite Size Mode=::NewEnumerator3
    Uniform Sprite Size=120
    UsePositionOffset=false
    Write Material Random=true
   )
   &DynamicMaterialParameters(
    Index 0 Param 1=2
    Index 0 Param 2=10
    Module.Index0Float0Write=true
    Module.Index0Float1Write=true
    Module.Index0Float2Write=true
    Module.Index0Float3Write=true
   )
   &CameraOffset()
  ]
  particleUpdateScript [
   &ParticleState(Kill Particles When Lifetime Has Elapsed=false)
   &SpriteRotationRate(Rotation Rate=120)
   &ScaleSpriteSize(Scale Sprite Size Mode=::NewEnumerator0, Uniform Scale Factor=$System.SizeAll)
   &SkeletalMeshLocation(Bone Sampling Mode=::NewEnumerator4, MaskSpawnGroup=false, Skeletal Mesh=$System.Skeletal Mesh Reader)
   &FN_Sprite_Attributes(
    Camera Offset=30
    Color=&MakeLinearColorFromVectorAndFloat(
     "Vector (RGB)"=&Multiply_Vector(A=&MakeVectorFromLinearColorRGB(Color=<rgba 3.7917 12.3035 40 2>), B=<1 1 1>)
     "Float (Alpha)"=&Multiply_Float(A=2, B=0.5)
    )
    Create Camera Offset Attribute=true
    Create Color Attribute=true
    Create SpriteSize Attribute=true
    Module Output:ENiagara_AttributeSamplingApplyOutput=::NewEnumerator1
    SpriteSize=&Multiply_Vector2D(
     A=&Vector2DFromFloat(Value=73)
     B=&Multiply_Vector2D_ByFloat(Float=$System.SizeAll, Vector2D=<1 1>)
    )
   )
   &FN_Mesh_Attributes(
    Color=&MakeLinearColorFromVectorAndFloat("Vector (RGB)"=<0 0 0>, "Float (Alpha)"=1)
    Create Color Attribute=true
    Create Scale Attribute=true
    Module Output=::NewEnumerator1
    Scale=&Multiply_VectorByFloat(Vector=&Multiply_VectorByFloat(Vector=<1 1 1>, Float=$System.SizeAll), Float=0.195)
   )
   &FN_Mesh_Attributes(
    Color=&MakeLinearColorFromVectorAndFloat("Vector (RGB)"=&MakeVectorFromLinearColorRGB(Color=<rgba 1.46 0.3768 1.5 1>), "Float (Alpha)"=3)
    Create Color Attribute=true
    Create Scale Attribute=true
    Module Output=::NewEnumerator1
    Scale=&Multiply_VectorByFloat(Vector=&Multiply_VectorByFloat(Vector=<2.5 2.5 0.2>, Float=$System.SizeAll), Float=1.8)
   ) {module=FN_Mesh_Attributes001}
  ]
  rendererClasses [&NiagaraSpriteRendererProperties, &NiagaraMeshRendererProperties]
  renderers [
   {rendererClass=&NiagaraSpriteRendererProperties}
   {rendererClass=&NiagaraSpriteRendererProperties, rendererIndex=1}
   {rendererClass=&NiagaraMeshRendererProperties, rendererIndex=2}
   {rendererClass=&NiagaraMeshRendererProperties, rendererIndex=3}
  ]
 }
 CenterRing {
  emitterUpdateScript [
   &EmitterState(Life Cycle Mode=::NewEnumerator1, Loop Behavior=::NewEnumerator1, Loop Duration Mode=::NewEnumerator1, UseLoopDelay=false)
   &SpawnBurst_Instantaneous(Use Spawn Probability=false, UseLoopCountLimit=false)
  ]
  particleSpawnScript [
   &InitializeParticle(
    Color=<rgba 0.8 0.8 0.8 1>
    Mesh Scale Mode=::NewEnumerator3
    Mesh Uniform Scale=0.57
    Position Offset=$System.FXOffset
    UsePositionOffset=true
    Write Material Random=true
   )
   &InitialMeshOrientation(Mesh Orientation Mode=::NewEnumerator2, Rotation=<0 0 -0.25>, Use Rotation Vector=true)
  ]
  particleUpdateScript [
   &CameraOffset(Camera Offset Amount=10)
   &ParticleState(Kill Particles When Lifetime Has Elapsed=false)
   &ScaleMeshSize(Scale Factor=&Multiply_VectorByFloat(Vector=<1 1 1>, Float=$System.SizeAll))
   &SkeletalMeshLocation(Bone Sampling Mode=::NewEnumerator4, MaskSpawnGroup=false, Skeletal Mesh=$System.Skeletal Mesh Reader)
   &DynamicMaterialParameters(
    Index 0 Param 1=&Multiply_Float(A=4.2858, B=$System.SizeAll)
    Module.Index0Float0Write=true
    Module.Index0Float1Write=true
    Module.Index0Float2Write=true
    Module.Index0Float3Write=true
   )
   &FN_Mesh_Attributes(
    Color=&MakeLinearColorFromVectorAndFloat("Vector (RGB)"=<1 1 1>, "Float (Alpha)"=1)
    Create Color Attribute=true
    Create Scale Attribute=true
    Module Output=::NewEnumerator1
    Scale=&Multiply_VectorByFloat(Vector=&Multiply_VectorByFloat(Vector=<1 1 1>, Float=$System.SizeAll), Float=0.5)
   )
   &ScaleColor(Scale Alpha=1.5, ScaleA=true, ScaleRGB=true)
   &Light_Attributes(Color=1, Color Mode=0, Falloff Exponent=1, Light Color:LinearColor=<rgba 15.4391 1.5625 30 1>, Light Exponent=3, Radius=1)
  ]
  rendererClasses [&NiagaraMeshRendererProperties, &NiagaraLightRendererProperties]
  renderers [
   {rendererClass=&NiagaraMeshRendererProperties, rendererIndex=0}
   {rendererClass=&NiagaraMeshRendererProperties, rendererIndex=1}
   {rendererClass=&NiagaraLightRendererProperties, rendererIndex=2}
  ]
 }
 ColorRing {
  emitterUpdateScript [
   &EmitterState(Life Cycle Mode=::NewEnumerator1, Loop Behavior=::NewEnumerator1, Loop Duration Mode=::NewEnumerator1, UseLoopDelay=false)
   &SpawnBurst_Instantaneous(Spawn Count=4, Use Spawn Probability=false, UseLoopCountLimit=false)
   &SpawnBurst_Instantaneous(Spawn Count=4, Spawn Time=5.7667, Use Spawn Probability=false, UseLoopCountLimit=false) {module=SpawnBurst_Instantaneous001}
  ]
  particleSpawnScript [
   &PartitionParticles(Number Of Partitions=4)
   &InitializeParticle(
    Color=<rgba 5 1.0482 2.4429 2>
    Lifetime Max=5
    Lifetime Min=3
    Lifetime Mode=::NewEnumerator1
    Mesh Scale Mode=::NewEnumerator3
    Mesh Uniform Scale=&SelectFloatFromArray(Direct Array Index=$Particles.PartitionParticles.Partition, Float Selection Array=DataInterface() {propertyValues {FloatData [4,3.8,4,3.8]}})
    Position Offset=&AddVectorToPosition(Position=Linked() {linkedVariable {name=System.FXOffset, type=Vector3f}}, Vector=<0 0 0>)
    UsePositionOffset=false
    Write Material Random=true
   )
   &InitialMeshOrientation(
    Mesh Orientation Mode=::NewEnumerator7
    Rotation=&SelectVectorFromArray(
     Direct Array Index=$Particles.PartitionParticles.Partition
     Vector Selection Array=DataInterface() {propertyValues {FloatData [<0 0 0>, <0 0 0.3>, <0 0 0.6>, <0 0 0.8>]}}
    )
    Use Rotation Vector=true
   )
   &DynamicMaterialParameters(
    Index 0 Param 1=&Multiply_Float(A=&RandomRangeFloat(Minimum=15, Maximum=25), B=1)
    Module.Index0Float0Write=true
    Module.Index0Float1Write=false
    Module.Index0Float2Write=false
    Module.Index0Float3Write=false
   )
   &CameraOffset(Camera Offset Amount=30)
  ]
  particleUpdateScript [
   &ParticleState(Kill Particles When Lifetime Has Elapsed=false)
   &SkeletalMeshLocation(Bone Sampling Mode=::NewEnumerator4, MaskSpawnGroup=false, Skeletal Mesh=$System.Skeletal Mesh Reader)
   &ScaleMeshSize(Scale Factor=&Multiply_VectorByFloat(
     Vector=&Multiply_VectorByFloat(Vector=<1 1 1>, Float=1)
     Float=&Multiply_Float(A=$System.SizeAll, B=$System.OrbitSize)
    ))
   &ScaleColor(
    Scale Alpha=&Multiply_Float(
     A=0.9
     B=&FN_LOD_Distance_Remap(
      Start Distance=300
      End Distance=400
      Minimum Output=1
      Maximum Output=&FN_FrontEndFloatModifier(Gameplay=0.4, Front End=0.6)
      Clamp Output=true
     )
    )
    Scale RGB=&Multiply_VectorByFloat(Vector=<1 1 1>, Float=0.5)
    ScaleA=true
    ScaleRGB=true
   )
  ]
  rendererClasses [&NiagaraMeshRendererProperties]
  renderers [{rendererClass=&NiagaraMeshRendererProperties, rendererIndex=0}]
 }
 Connection {
  emitterUpdateScript [
   &EmitterState(Life Cycle Mode=::NewEnumerator1, Loop Behavior=::NewEnumerator1, Loop Duration Mode=::NewEnumerator1, UseLoopDelay=false)
   &SpawnBurst_Instantaneous(Use Spawn Probability=false, UseLoopCountLimit=false)
   /SetVariables(Emitter.PlaneWidth:NiagaraFloat=&Multiply_Float(A=20, B=$System.SizeAll)) {isSetParameters=true, module=SetVariables_#5}
  ]
  particleSpawnScript [
   &InitializeParticle(
    Color=<rgba 1 1 1 2>
    Mesh Scale=&Multiply_VectorByFloat(Vector=<0.9 1 1>, Float=3.6429)
    Mesh Scale Mode=::NewEnumerator4
    UsePositionOffset=false
    Write Material Random=true
   )
   &InitialMeshOrientation(Mesh Orientation Mode=::NewEnumerator7, Rotation=<0 0 0.25>, Use Rotation Vector=true)
  ]
  particleUpdateScript [
   &ParticleState(Kill Particles When Lifetime Has Elapsed=false)
   &ScaleMeshSize(Scale Factor=&Multiply_VectorByFloat(Vector=<1 1 1>, Float=$System.SizeAll))
   &SkeletalMeshLocation(Bone Sampling Mode=::NewEnumerator4, MaskSpawnGroup=false, Skeletal Mesh=$System.Skeletal Mesh Reader)
   &FN_Mesh_Attributes(
    Color=&MakeLinearColorFromVectorAndFloat(
     "Vector (RGB)"=&Multiply_Vector(A=&MakeVectorFromLinearColorRGB(Color=<rgba 1 1 1 1>), B=<1 1 1>)
     "Float (Alpha)"=&Multiply_Float(A=3, B=1)
    )
    Create Color Attribute=true
    Create Scale Attribute=true
    Module Output=::NewEnumerator1
    Scale=&Multiply_Vector(
     A=&Multiply_VectorByFloat(Vector=&Multiply_VectorByFloat(Vector=<0.9 1 1>, Float=$System.SizeAll), Float=3.5715)
     B=<1 1 1>
    )
   )
  ]
  rendererClasses [&NiagaraMeshRendererProperties]
  renderers [
   {rendererClass=&NiagaraMeshRendererProperties, rendererIndex=0}
   {rendererClass=&NiagaraMeshRendererProperties, rendererIndex=1}
  ]
 }
 ConnectionFacingCamera {
  emitterUpdateScript [
   &EmitterState(Life Cycle Mode=::NewEnumerator1, Loop Behavior=::NewEnumerator1, Loop Duration Mode=::NewEnumerator1, UseLoopDelay=false)
   &SpawnBurst_Instantaneous(Use Spawn Probability=false, UseLoopCountLimit=false)
   /SetVariables(Emitter.PlaneWidth:NiagaraFloat=&Multiply_Float(A=20, B=$System.SizeAll)) {isSetParameters=true, module=SetVariables_#5}
  ]
  particleSpawnScript [
   &InitializeParticle(
    Color=<rgba 1 1 1 3>
    Mesh Scale=&Multiply_VectorByFloat(Vector=<0.9 1 1>, Float=3.6429)
    Mesh Scale Mode=::NewEnumerator4
    UsePositionOffset=false
    Write Material Random=true
   )
   &InitialMeshOrientation(Mesh Orientation Mode=::NewEnumerator7, Rotation=<0 0 0.25>, Use Rotation Vector=true)
   &InitialMeshOrientation(Mesh Orientation Mode=::NewEnumerator7, Use Rotation Vector=true) {enabled=false, module=InitialMeshOrientation001}
  ]
  particleUpdateScript [
   &ParticleState(Kill Particles When Lifetime Has Elapsed=false)
   &ScaleMeshSize(Scale Factor=&Multiply_VectorByFloat(Vector=<1 1 1>, Float=$System.SizeAll))
   &SkeletalMeshLocation(Bone Sampling Mode=::NewEnumerator4, MaskSpawnGroup=false, Skeletal Mesh=$System.Skeletal Mesh Reader)
   &FN_Mesh_Attributes(
    Color=&MakeLinearColorFromVectorAndFloat(
     "Vector (RGB)"=&Multiply_Vector(A=&MakeVectorFromLinearColorRGB(Color=<rgba 1 1 1 1>), B=<1 1 1>)
     "Float (Alpha)"=&Multiply_Float(A=3, B=1)
    )
    Create Color Attribute=true
    Create Scale Attribute=true
    Module Output=::NewEnumerator1
    Scale=&Multiply_Vector(
     A=&Multiply_VectorByFloat(Vector=&Multiply_VectorByFloat(Vector=<0.9 1 1>, Float=$System.SizeAll), Float=3.5715)
     B=<1 1 1>
    )
   )
   &ScaleColor(ScaleA=true, ScaleRGB=true)
  ]
  rendererClasses [&NiagaraMeshRendererProperties]
  renderers [
   {rendererClass=&NiagaraMeshRendererProperties, rendererIndex=0}
   {rendererClass=&NiagaraMeshRendererProperties, rendererIndex=1}
  ]
 }
 Orbit {
  emitterUpdateScript [
   &EmitterState(Life Cycle Mode=::NewEnumerator1, Loop Behavior=::NewEnumerator1, Loop Duration Mode=::NewEnumerator1, UseLoopDelay=false)
   &SpawnBurst_Instantaneous(Use Spawn Probability=false, UseLoopCountLimit=false)
   /SetVariables(Emitter.CustomSize:Vector3f=<1 1 0.8>) {isSetParameters=true}
  ]
  particleSpawnScript [
   &InitializeParticle(
    Color=<rgba 1.5 1.5 1.5 1>
    Mesh Scale Mode=::NewEnumerator3
    Mesh Uniform Scale=5
    Position Offset=$System.FXOffset
    UsePositionOffset=true
    Write Material Random=true
   )
   &InitialMeshOrientation(Mesh Orientation Mode=::NewEnumerator2, Rotation=<0 0 -0.25>, Use Rotation Vector=true)
  ]
  particleUpdateScript [
   &ParticleState(Kill Particles When Lifetime Has Elapsed=false)
   &ScaleMeshSize(Scale Factor=&Multiply_VectorByFloat(
     Vector=&Multiply_Vector(A=<1 1 2>, B=$Emitter.CustomSize)
     Float=&Multiply_Float(A=$System.SizeAll, B=$System.OrbitSize)
    ))
   &SkeletalMeshLocation(Bone Sampling Mode=::NewEnumerator4, MaskSpawnGroup=false, Skeletal Mesh=$System.Skeletal Mesh Reader)
   &DynamicMaterialParameters(
    Index 0 Param 1=2.5
    Module.Index0Float0Write=true
    Module.Index0Float1Write=true
    Module.Index0Float2Write=true
    Module.Index0Float3Write=true
   )
   &FN_Mesh_Attributes(
    Color=&MakeLinearColorFromVectorAndFloat(
     "Vector (RGB)"=&Multiply_Vector(A=&MakeVectorFromLinearColorRGB(Color=<rgba 2.5 2.5 2.5 1>), B=<1 1 1>)
     "Float (Alpha)"=&Multiply_Float(A=1.3, B=1)
    )
    Create Color Attribute=true
    Create Scale Attribute=true
    Module Output=::NewEnumerator1
    Scale=&Multiply_Vector(
     A=&Multiply_VectorByFloat(
      Vector=&Multiply_VectorByFloat(Vector=$Emitter.CustomSize, Float=&Multiply_Float(A=$System.SizeAll, B=$System.OrbitSize))
      Float=4.5
     )
     B=<1 1 2>
    )
   )
   &FN_Mesh_Attributes(
    Color=&MakeLinearColorFromVectorAndFloat(
     "Vector (RGB)"=&Multiply_Vector(A=&MakeVectorFromLinearColorRGB(Color=<rgba 0.0697 0.0045 0.2 4>), B=<1 1 1>)
     "Float (Alpha)"=&Multiply_Float(A=4, B=1)
    )
    Create Color Attribute=true
    Create Scale Attribute=true
    Module Output=::NewEnumerator1
    Scale=&Multiply_Vector(
     A=&Multiply_VectorByFloat(
      Vector=&Multiply_VectorByFloat(Vector=$Emitter.CustomSize, Float=&Multiply_Float(A=$System.SizeAll, B=$System.OrbitSize))
      Float=4
     )
     B=<1 1 0.1>
    )
   ) {module=FN_Mesh_Attributes001}
   &FN_Mesh_Attributes(
    Color=&MakeLinearColorFromVectorAndFloat(
     "Vector (RGB)"=&Multiply_Vector(A=&MakeVectorFromLinearColorRGB(Color=<rgba 0.8 0.8 0.8 1>), B=<1 1 1>)
     "Float (Alpha)"=&Multiply_Float(A=1, B=1)
    )
    Create Color Attribute=true
    Create Scale Attribute=true
    Module Output=::NewEnumerator1
    Scale=&Multiply_Vector(
     A=&Multiply_VectorByFloat(
      Vector=&Multiply_VectorByFloat(Vector=$Emitter.CustomSize, Float=&Multiply_Float(A=$System.SizeAll, B=$System.OrbitSize))
      Float=5
     )
     B=<1 1 2>
    )
   ) {module=FN_Mesh_Attributes002}
  ]
  rendererClasses [&NiagaraMeshRendererProperties]
  renderers [
   {rendererClass=&NiagaraMeshRendererProperties, rendererIndex=0}
   {rendererClass=&NiagaraMeshRendererProperties, rendererIndex=1}
   {rendererClass=&NiagaraMeshRendererProperties, rendererIndex=2}
   {rendererClass=&NiagaraMeshRendererProperties, rendererIndex=3}
  ]
 }
 Smoke {
  emitterUpdateScript [
   &EmitterState(Life Cycle Mode=::NewEnumerator1, Loop Duration=1, UseLoopDelay=false)
   &SpawnRate(SpawnRate=5, Use Spawn Probability=false)
   &SpawnBurst_Instantaneous(Spawn Count=3, Use Spawn Probability=false, UseLoopCountLimit=false)
  ]
  particleSpawnScript [
   &InitializeParticle(
    Color=&MakeCustomLinearColorFromBool(Bool=&RandomBool(), True Color=<rgba 0.4618 0.1 2 1>, False Color=<rgba 0.9891 0.0481 1 1>)
    Lifetime Mode=::NewEnumerator1
    Sprite Rotation Mode=::NewEnumerator2
    Sprite Size Mode=::NewEnumerator1
    Uniform Sprite Size Max=80
    Uniform Sprite Size Min=60
    UsePositionOffset=false
    Write Material Random=true
   )
   &ShapeLocation(
    MaskSpawnGroup=false
    Non Uniform Scale=<1 1 0.01>
    Offset Mode=::NewEnumerator0
    Override Seed=false
    Rotation Mode=::NewEnumerator2
    Shape Origin=&VectorFromFloat(Value=0)
    Sphere Radius=&Multiply_Float(
     A=&RandomRangeFloat(Minimum=60, Maximum=70)
     B=&FN_FrontEndFloatModifier(Gameplay=1, Front End=1)
    )
    Sphere Surface Distribution=1
    UseOwnerScale=true
   )
   &DynamicMaterialParameters(
    Index 0 Param 1=&MakeCustomFloatFromBool(Bool=&RandomBool(), True Float=0, False Float=0.75)
    Module.Index0Float0Write=true
    Module.Index0Float1Write=true
    Module.Index0Float2Write=true
    Module.Index0Float3Write=true
   )
  ]
  particleUpdateScript [
   &SkeletalMeshLocation(
    Bone Sampling Mode=::NewEnumerator4
    MaskSpawnGroup=false
    Position Sampling=::NewEnumerator1
    Sampled Position Offset=&AddVectorToPosition(Position=$Particles.ShapeLocation.ShapePosition, Vector=<0 0 10>)
    Skeletal Mesh=$System.Skeletal Mesh Reader
   )
   &ParticleState()
   &ScaleSpriteSize(
    Scale Sprite Size Mode=::NewEnumerator0
    Uniform Scale Factor=&FN_FrontEndFloatModifier(Gameplay=1, Front End=0.7)
   )
   &VortexForce(
    UseOriginOffset=false
    Vortex Force Amount=1000
    Vortex Origin:NiagaraPosition=&AddVectorToPosition(Position=$Engine.Emitter.SimulationPosition, Vector=$System.FXOffset)
   )
   &ScaleVelocity()
   /SetVariables(Transient.PhysicsForce:Vector3f=&Multiply_Vector(A=$Transient.PhysicsForce, B=<1 1 1>)) {isSetParameters=true, module=SetVariables_#6}
   &PointAttractionForce(
    Attractor Position:NiagaraPosition=&AddVectorToPosition(Position=$Particles.SkeletalMeshLocation.SampledPosition, Vector=$System.FXOffset)
    Kill Within Radius=false
    Module.Use Falloff=false
    UseAttractorOffset=false
   )
   &Drag(Drag=10, Rotational Drag=10, Use Linear Drag=true, Use Rotational Drag=true)
   &SolveForcesAndVelocity(Clamp Velocity=false, Limit Acceleration=false)
   &CameraOffset(Camera Offset Amount=5)
   &ScaleColor(
    Scale Alpha=&FloatFromCurve(
     FloatCurve=DataInterface() {propertyValues {Curve {
        defaultValue=1
        keys [
         {interpMode=RCIM_Cubic, tangentMode=RCTM_User, time=0, value=0}
         {interpMode=RCIM_Cubic, tangentMode=RCTM_User, time=0.5, value=1}
         {tangentMode=RCTM_User, time=1, value=0}
        ]
       }}}
     Scale Curve=&Multiply_Float(
      A=&RandomRangeFloat(Minimum=0.5, Maximum=1)
      B=&FN_FrontEndFloatModifier(Gameplay=1, Front End=0)
     )
    )
    ScaleA=true
    ScaleRGB=true
   )
  ]
  rendererClasses [&NiagaraSpriteRendererProperties]
  renderers [{rendererClass=&NiagaraSpriteRendererProperties}]
 }
 SmokeWorldSpace {
  emitterUpdateScript [
   &EmitterState(Life Cycle Mode=::NewEnumerator1, Loop Duration=1, UseLoopDelay=false)
   &SpawnPerUnit(
    Module.Use Max Movement Threshold=true
    Module.Use Movement Tolerance=false
    Spawn Probability:NiagaraFloat=&MakeCustomFloatFromBool(
     Bool=&SetBoolByFloatComparison(A=&VectorLength(Vector=$Engine.Owner.Velocity), B=1000)
     True Float=0
     False Float=&FN_FrontEndFloatModifier(Gameplay=0.5, Front End=0)
    )
    Spawn Spacing=10
    Use Spawn Probability=true
   )
  ]
  particleSpawnScript [
   &InitializeParticle(
    Color=&LerpLinearColors(
     StartColor=<rgba 0.6198 0.06 3 1>
     EndColor=<rgba 0.7917 0.08 0.8 1>
     LerpFactor=&UniformAOrBFloat(A=0, Distribution Weight=0.7, Evaluation Type=::NewEnumerator0, Recalculate Random Each Loop=true, Override Seed=false)
    )
    Lifetime Max=1
    Lifetime Mode=::NewEnumerator1
    Sprite Rotation Mode=::NewEnumerator2
    Sprite Size Mode=::NewEnumerator1
    Uniform Sprite Size Max=60
    Uniform Sprite Size Min=50
    UsePositionOffset=false
    Write Material Random=true
   )
   &SkeletalMeshLocation(
    Bone Sampling Mode=::NewEnumerator4
    MaskSpawnGroup=false
    Position Sampling=::NewEnumerator1
    Sampled Position Offset=<0 -40 10>
    Skeletal Mesh=$System.Skeletal Mesh Reader
   )
   &ShapeLocation(
    MaskSpawnGroup=false
    Non Uniform Scale=<1 1 0.01>
    Offset Mode=::NewEnumerator0
    Override Seed=false
    Rotation Mode=::NewEnumerator2
    Sphere Radius=40
    Sphere Surface Distribution=1
    UseOwnerScale=true
   )
   &DynamicMaterialParameters(
    Index 0 Param 1=&MakeCustomFloatFromBool(Bool=&RandomBool(), True Float=0, False Float=0.75)
    Module.Index0Float0Write=true
    Module.Index0Float1Write=true
    Module.Index0Float2Write=true
    Module.Index0Float3Write=true
   )
   &CameraOffset(Camera Offset Amount=5)
  ]
  particleUpdateScript [
   &ParticleState()
   &SpriteRotationRate(Rotation Rate=&RandomRangeFloat(Minimum=-100, Maximum=100))
   &InheritSourceMovement(Applied Position Scale=<0.2 0.2 0.2>, ApplyForce=false, ApplyPosition=true, ApplyPositionFalloff=false)
   &SolveForcesAndVelocity(Clamp Velocity=false, Limit Acceleration=false)
   &FN_Sprite_Attributes(
    Color=&MakeLinearColorFromVectorAndFloat(
     "Vector (RGB)"=&Multiply_Vector(
      A=&MakeVectorFromLinearColorRGB(Color=&LerpLinearColors(
        StartColor=<rgba 0.6198 0.06 3 1>
        EndColor=<rgba 0.7917 0.08 0.8 1>
        LerpFactor=&UniformAOrBFloat(A=0, Distribution Weight=0.8, Evaluation Type=::NewEnumerator0, Recalculate Random Each Loop=true, Override Seed=false)
       ))
      B=&Multiply_VectorByFloat(Vector=<1 1 1>, Float=0.5)
     )
     "Float (Alpha)"=&Multiply_Float(
      A=&FloatFromCurve(
       FloatCurve=DataInterface() {propertyValues {Curve {
          defaultValue=1
          keys [
           {interpMode=RCIM_Cubic, tangentMode=RCTM_User, time=0, value=0}
           {interpMode=RCIM_Cubic, tangentMode=RCTM_User, time=0.301, value=1.0471}
           {tangentMode=RCTM_User, time=1, value=0}
          ]
         }}}
       Scale Curve=&FN_LOD_Distance_Remap(Start Distance=300, End Distance=400, Minimum Output=1, Maximum Output=0, Clamp Output=true)
      )
      B=1.2
     )
    )
    Create Color Attribute=true
    Create SpriteSize Attribute=true
    Module Output:ENiagara_AttributeSamplingApplyOutput=::NewEnumerator1
    SpriteSize=&Multiply_Vector2D_ByFloat(Float=&RandomRangeFloat(Minimum=100, Maximum=80), Vector2D=<1 1>)
   )
   &ScaleColor(
    Scale Alpha=&FloatFromCurve(
     FloatCurve=DataInterface() {propertyValues {Curve {
        defaultValue=1
        keys [
         {interpMode=RCIM_Cubic, tangentMode=RCTM_User, time=0, value=0}
         {interpMode=RCIM_Cubic, tangentMode=RCTM_User, time=0.301, value=1.0471}
         {tangentMode=RCTM_User, time=1, value=0}
        ]
       }}}
     Scale Curve=&Multiply_Float(
      A=0.3
      B=&FN_LOD_Distance_Remap(Start Distance=400, End Distance=550, Minimum Output=0, Maximum Output=1, Clamp Output=true)
     )
    )
    Scale RGB=&Multiply_VectorByFloat(Vector=<1 1 1>, Float=1.5)
    ScaleA=true
    ScaleRGB=true
   )
   /SetVariables(Particles.VisibilityTag=&MakeIntFromFloat(Float=&FN_LOD_Distance_Remap(Start Distance=300, End Distance=400, Minimum Output=0, Maximum Output=1, Clamp Output=true))) {isSetParameters=true}
  ]
  rendererClasses [&NiagaraSpriteRendererProperties]
  renderers [
   {rendererClass=&NiagaraSpriteRendererProperties}
   {rendererClass=&NiagaraSpriteRendererProperties, rendererIndex=1}
  ]
 }
 Stone {
  simTarget=GPUComputeSim
  emitterUpdateScript [
   &EmitterState(Life Cycle Mode=::NewEnumerator1, Loop Duration=1, UseLoopDelay=false)
   &SpawnRate(SpawnRate=9, Use Spawn Probability=false)
   &SpawnBurst_Instantaneous(Spawn Count=10, Use Spawn Probability=false, UseLoopCountLimit=false)
  ]
  particleSpawnScript [
   &InitializeParticle(
    Color=<rgba 0.7734 0.7656 1 1>
    Lifetime Max=3
    Lifetime Min=3
    Lifetime Mode=::NewEnumerator1
    Mesh Scale Mode=::NewEnumerator1
    Mesh Uniform Scale Max=4
    Mesh Uniform Scale Min=2
    Sprite Size Mode=::NewEnumerator3
    UsePositionOffset=false
    Write Material Random=true
   )
   &ShapeLocation(
    MaskSpawnGroup=false
    Non Uniform Scale=<1 1 0.2>
    Offset Mode=::NewEnumerator0
    Override Seed=false
    Rotation Mode=::NewEnumerator2
    Shape Origin=&VectorFromFloat(Value=0)
    Sphere Radius=80
    Sphere Surface Distribution=1
    UseOwnerScale=true
   )
   &InitialMeshOrientation(Use Rotation Vector=false)
   &CameraOffset() {module=CameraOffset001}
  ]
  particleUpdateScript [
   &SkeletalMeshLocation(
    Bone Sampling Mode=::NewEnumerator4
    MaskSpawnGroup=false
    Position Sampling=::NewEnumerator1
    Sampled Position Offset:NiagaraPosition=$Particles.ShapeLocation.ShapePosition
    Skeletal Mesh=$System.Skeletal Mesh Reader
   )
   &ParticleState()
   &ScaleMeshSize(Scale Factor=&Multiply_VectorByFloat(
     Vector=&RandomRangeVector(Minimum=<0.5 0.5 0.5>, Maximum=<1.2 1 1>)
     Float=&Multiply_Float(
      A=&Multiply_Float(
       A=0.12
       B=&FloatFromCurve(FloatCurve=DataInterface() {propertyValues {Curve {
           defaultValue=1
           keys [
            {interpMode=RCIM_Cubic, tangentMode=RCTM_User, time=0, value=0}
            {arriveTangent=1.4286, interpMode=RCIM_Cubic, leaveTangent=1.4286, time=0.3, value=1}
            {interpMode=RCIM_Cubic, tangentMode=RCTM_User, time=0.7, value=1}
            {tangentMode=RCTM_User, time=1, value=0}
           ]
          }}})
      )
      B=1
     )
    ))
   &VortexForce(
    UseOriginOffset=false
    Vortex Force Amount=1000
    Vortex Origin:NiagaraPosition=&AddVectorToPosition(Position=$Engine.Emitter.SimulationPosition, Vector=$System.FXOffset)
   )
   &ScaleVelocity()
   /SetVariables(Transient.PhysicsForce:Vector3f=&Multiply_Vector(A=$Transient.PhysicsForce, B=<1 1 1>)) {isSetParameters=true, module=SetVariables_#6}
   &PointAttractionForce(
    Attractor Position:NiagaraPosition=&AddVectorToPosition(Position=$Particles.SkeletalMeshLocation.SampledPosition, Vector=$System.FXOffset)
    Kill Within Radius=false
    Module.Use Falloff=false
    UseAttractorOffset=false
   )
   &Drag(Drag=10, Rotational Drag=10, Use Linear Drag=true, Use Rotational Drag=true)
   &SolveForcesAndVelocity(Clamp Velocity=false, Limit Acceleration=false)
   &CameraOffset(Camera Offset Amount=5)
  ]
  rendererClasses [&NiagaraMeshRendererProperties]
  renderers [{rendererClass=&NiagaraMeshRendererProperties, rendererIndex=0}]
 }
}
summary {
 systemName=NS_Emote_SkyRuled_Main
 emitters [
  {emitterName=CenterRing, rendererClasses [&NiagaraMeshRendererProperties, &NiagaraLightRendererProperties]}
  {
   bEnabled=false
   emitterName=BlackHole
   rendererClasses [&NiagaraSpriteRendererProperties, &NiagaraMeshRendererProperties]
  }
  {emitterName=Orbit, rendererClasses [&NiagaraMeshRendererProperties]}
  {emitterName=Connection, rendererClasses [&NiagaraMeshRendererProperties]}
  {emitterName=ColorRing, rendererClasses [&NiagaraMeshRendererProperties]}
  {emitterName=Stone, simTarget=GPUComputeSim, rendererClasses [&NiagaraMeshRendererProperties]}
  {emitterName=ConnectionFacingCamera, rendererClasses [&NiagaraMeshRendererProperties]}
  {emitterName=Smoke, rendererClasses [&NiagaraSpriteRendererProperties]}
  {emitterName=SmokeWorldSpace, rendererClasses [&NiagaraSpriteRendererProperties]}
  {emitterName=BlackHole001, rendererClasses [&NiagaraSpriteRendererProperties, &NiagaraMeshRendererProperties]}
 ]
 userVariables [
  User.bIsFrontEnd {type=NiagaraBool}
  User.bIsFrontEndPreview {type=NiagaraBool}
 ]
}
